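# Translucent unlit beam-trail material for the Snowman pickaxe melee trail, colored by particle color and using normal-mapped refraction.
# Two panning smoke textures drive a bump offset that distorts either a time-of-day-adjusted splash mask or cracks multiplied by swirl, chosen by static switch.
# Sine UV falloff, particle alpha and depth fade shape opacity.
# A dynamic parameter scales refraction, while mobile skips both depth fade and refraction.
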Material /BRCosmetics/Weapons/FORT_Melee/Pickaxe_Snowman/FX/M_Melee_Snowman_Trail
graph {
 properties {
  bUsedWithBeamTrails=true
  blendMode=BLEND_Translucent
  pixelDepthOffsetMode=PDOM_Legacy
  refractionMethod=RM_IndexOfRefraction
  shadingModel=MSM_Unlit
  usedShadingModels=Unlit
 }
 nodes [
  LinearInterpolate(
   A=Constant3Vector() {constant=<rgba 0 0 1 0>}
   B=TextureSample() {samplerType=SAMPLERTYPE_Normal, texture=/Game/Effects/Fort_Effects/Textures/Husk/Poison/T_PoisonTile_N}.RGB
   Alpha=ScalarParameter() {defaultValue=3, parameterName=Normal_Intensity}
  )
  n1 = ParticleColor()
  MaterialFunctionCall(Emissive=@n1.RGB) {materialFunction=&MF_FX_EmissiveAdjustments}
  n2 = TextureSample(UVs=Panner(Coordinate=TextureCoordinate()) {speedX=0.5}) {texture=&T_Inky_Smoke_Tile}
  n3 = TextureSample(UVs=Panner(Coordinate=TextureCoordinate()) {speedX=0.3}) {texture=/Game/Global/Textures/Default/Tileable_Patterns/T_Smoke_Puff_Tile_Up}
  n4 = BumpOffset(Height=Multiply(A=Multiply(A=@n2.R, B=@n3.R), B=ScalarParameter() {defaultValue=10}))
  n5 = Multiply(A=Add(A=TextureCoordinate() {uTiling=3}, B=@n4)) {constB=0.5}
  n6 = Multiply(A=TextureSample(UVs=Panner(Coordinate=@n5) {speedX=1}) {samplerSource=SSM_Wrap_WorldGroupSettings, texture=/Game/Effects/Fort_Effects/Textures/Water/T_WaterFallSplash_Mask}.R)
  n7 = LinearInterpolate(A=Round(None=@n6), B=@n6) {constAlpha=0.4}
  n8 = Multiply(A=@n7)
  n9 = MaterialFunctionCall(Color_Morning=@n8, Color_Day=@n8, Color_Evening=@n8, Color_Night=Multiply(A=@n7) {constB=0.3}) {materialFunction=&Color_Time_of_Day}
  n10 = TextureCoordinate() {uTiling=9, vTiling=0.1}
  n11 = MaterialFunctionCall(XY=Multiply(A=Add(A=@n10, B=@n4)) {constB=0.5}) {materialFunction=&Swizzle}
  n12 = Multiply(A=TextureSample(UVs=Panner(Coordinate=@n11.YX) {speedY=1}) {samplerSource=SSM_Wrap_WorldGroupSettings, texture=/Game/Effects/Fort_Effects/Textures/Tile/T_Glass_Cracks_M}.R) {constB=30}
  n13 = TextureCoordinate() {vTiling=0.1}
  n14 = MaterialFunctionCall(XY=Multiply(A=Add(A=@n13, B=@n4)) {constB=0.5}) {materialFunction=&Swizzle}
  n15 = Constant() {r=0.3}
  n16 = Multiply(
   A=LinearInterpolate(A=Round(None=@n12), B=@n12) {constAlpha=1}
   B=TextureSample(UVs=Panner(Coordinate=@n14.YX, Time=@n15) {speedY=0.5}) {samplerSource=SSM_Wrap_WorldGroupSettings, texture=/Game/Effects/Fort_Effects/Textures/T_Swril_Goo_Tile}.G
  )
  n17 = MaterialFunctionCall(Float2=TextureCoordinate() {vTiling=0.99}) {materialFunction=&BreakOutFloat2Components}
  n18 = Multiply(A=Power(Base=Sine(None=@n17.R) {period=2}), B=Power(Base=Sine(None=@n17.G) {period=2}) {constExponent=2.5})
  n19 = Multiply(A=@n1.A, B=Multiply(A=StaticSwitchParameter(True=@n9.Result, False=@n16) {parameterName=Use_Texture_A}, B=@n18))
  n20 = DepthFade(Opacity=@n19, FadeDistance=ScalarParameter() {defaultValue=20, parameterName=Depth_Fade})
  ShadingPathSwitch(Default=@n20, Deferred=@n20, Mobile=@n19) {desc=Migrated from FeatureLevelSwitch}
  n21 = LinearInterpolate(Alpha=Multiply(A=@n18, B=DynamicParameter() {defaultValue=<rgba 0.5 1 1 1>, paramNames [Reflection_intensity, Param2, Param3, Param4]}.Reflection_intensity)) {constA=1, constB=1.05}
  ShadingPathSwitch(Default=@n21, Deferred=@n21, Mobile=Constant() {r=1}) {desc=Migrated from FeatureLevelSwitch}
 ]
}
